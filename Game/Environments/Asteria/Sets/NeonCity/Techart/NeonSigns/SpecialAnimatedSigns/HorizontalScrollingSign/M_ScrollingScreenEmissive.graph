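# Emissive material for a horizontally scrolling LCD-style neon sign screen in the NeonCity set.
# Pixelated UVs, snapped by QuantizeAmount, feed two panned samples of a gradient texture: one gives pulse color, the other a scrolling alpha mask posterized into steps.
# The masked color is multiplied by a vertical gradient and tiled LCD panel texture, then passed through MF_NeonCityEmissiveBalance and dimmed at grazing angles by inverted Fresnel.
# A glass grime texture adds subtle surface variation.
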Material /Game/Environments/Asteria/Sets/NeonCity/Techart/NeonSigns/SpecialAnimatedSigns/HorizontalScrollingSign/M_ScrollingScreenEmissive
graph {
 properties {bUsedWithNanite=true, pixelDepthOffsetMode=PDOM_Legacy, usedShadingModels=Default Lit}
 nodes [
  Constant3Vector()
  Constant()
  Constant() {r=0.1}
  Constant() {r=0.66}
  LinearInterpolate(Alpha=TextureSample() {samplerType=SAMPLERTYPE_LinearColor, texture=/Game/Environments/Apollo/Sets/Tinfoil/Textures/T_Tinfoil_GlassGrime_M}.G) {constA=0.1, constB=0.3}
  n1 = TextureCoordinate()
  Multiply(A=@n1)
  n2 = TextureSample(UVs=Panner(Coordinate=NamedRerouteUsage()) {speedX=0.05}) {texture=&/QuantizedNeonCityGradient_M}
  n3 = Multiply(A=@n2.RGB, B=ScalarParameter() {defaultValue=2, parameterName=PulseBrightness})
  n4 = ScalarParameter() {defaultValue=-0.33, parameterName=Scroll X}
  n5 = ScalarParameter() {parameterName=Scroll Y}
  n6 = TextureSample(UVs=Panner(Coordinate=NamedRerouteUsage(), Time=Time(), Speed=AppendVector(A=@n4, B=@n5))) {texture=&QuantizedNeonCityGradient_M}
  n7 = Clamp(None=Power(Base=@n6.A) {constExponent=1.5})
  n8 = Divide(A=Round(None=Multiply(A=@n7) {constB=20})) {constB=20}
  n9 = TextureCoordinate() {uTiling=400, vTiling=400}
  n10 = AppendVector(A=ScalarParameter() {defaultValue=2, parameterName=X}, B=ScalarParameter() {defaultValue=0.2, parameterName=Y})
  n11 = Multiply(
   A=Multiply(
    A=LinearInterpolate(B=@n3, Alpha=@n8)
    B=TextureSample() {samplerType=SAMPLERTYPE_LinearColor, texture=/Game/Global/Textures/Default/Masks_and_Gradients/T_VertGradient}.G
   )
   B=TextureSample(UVs=Multiply(A=@n9, B=@n10)) {samplerType=SAMPLERTYPE_Masks, texture=/Game/Environments/Asteria/Sets/NeonCity/Techart/Textures/T_LCD_Panels_LCDSoft}.RGB
  )
  n12 = ScalarParameter() {defaultValue=2, parameterName=Exponent Fresnel}
  Multiply(
   A=MaterialFunctionCall!(Emissive=Multiply(A=@n11) {constB=2}) {materialFunction=/Game/Environments/Asteria/Sets/NeonCity/Techart/Screen/MaterialFunctions/MF_NeonCityEmissiveBalance}.Result
   B=OneMinus(None=Fresnel(ExponentIn=@n12) {baseReflectFraction=0})
  )
  Multiply(A=@n2.RGB) {constB=0.1}
  n13 = ScalarParameter() {defaultValue=45, parameterName=QuantizeAmount}
  n14 = Multiply(A=@n1, B=@n13) {constB=20}
  NamedRerouteDeclaration(None=Divide(A=Round(None=@n14), B=@n13) {constB=20}) {name=QuanitizedUVs, nodeColor=<rgba 0.5938 0 1 1>}
 ]
}
